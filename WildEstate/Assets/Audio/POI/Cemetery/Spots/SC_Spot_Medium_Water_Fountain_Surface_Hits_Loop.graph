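# Looping ambient Sound Cue for the water surface hits of a medium fountain spot in the Cemetery POI.
# A Random node picks one of two equally weighted looping waves, and a Modulator slightly varies its volume.
# An indefinitely looping Enveloper adds a gentle, slow volume dip and recovery.
# The cue uses the outdoor ambient sound class, a Large attenuation borrowed from Domestic_Commercial spots, and a two-max, prevent-new loop concurrency.
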SoundCue /WildEstate/Assets/Audio/POI/Cemetery/Spots/SC_Spot_Medium_Water_Fountain_Surface_Hits_Loop
root=Enveloper_0
nodes [
 {
  id=Enveloper_0
  children [Modulator_0]
  properties {
   bLoop=true
   bLoopIndefinitely=true
   loopEnd=20
   pitchCurve {editorCurveData {keys [{time=0}]}}
   volumeCurve {editorCurveData {keys [<at -0.0028 1>,<at 0 1>,<at 10 0.95>,<at 20 1>]}}
  }
 }
 {id=Modulator_0, children [Random_1], properties {volumeMax=1, volumeMin=0.9}}
 {id=Random_1, children [WavePlayer_0,WavePlayer_1], properties {weights [1,1]}}
 {
  id=WavePlayer_0
  properties {bLooping=true, soundWaveAssetPtr=&/Spot_Medium_Water_Fountain_Surface_Hits_Loop_01}
 }
 {
  id=WavePlayer_1
  properties {bLooping=true, soundWaveAssetPtr=&/Spot_Medium_Water_Fountain_Surface_Hits_Loop_02}
 }
]
properties {
 attenuationSettings=/WildEstate/Assets/Audio/POI/Domestic_Commercial/Spots/Spot_WildEstate_Domestic_Commercial_Large_Att
 bDebug=true
 maxDistance=2300
 soundClassObject=&Fort_Ambient_Outdoors_Athena
 volumeMultiplier=1.1
 attenuationOverrides {bSpatialize=false, stereoSpread=1147.9967, attenuationShapeExtents {x=6826.1992}}
 concurrencySet [/AmbientAudioBase/Concurrency/AmbientConc_2Max_Loop_PreventNew_Default_Concurrency]
}
